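# Material function that stretches screen-aligned particle sprites along their velocity into ellipsoids, via World Position Offset.
# Particle speed remapped between min and max scale factors (overridable) scales the direction, flattened to camera XY.
# The quad is rebuilt from UVs and particle size, then stretched by a dot-product projection, so distortion vanishes when motion aligns with the view.
# Also outputs world-space normals and an opacity mask, defaulting to sphere textures.
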MaterialFunction /Engine/Functions/Engine_MaterialFunctions02/WorldPositionOffset/Sprite_Ellipsoid
graph {
 properties {
  bExposeToLibrary=true
  description="The material function will align a particle, or plane, to a normalized velocity vector, and scale on that axis while maintaining it's screen-alignment. The distortion along the velocity vector disappears when the vector and the viewer's line of sight align. The module can be used to represent velocity aligned ellipsoidal shapes on quads."
  libraryCategoriesText [World Position Offset]
 }
 nodes [
  FunctionInput() {
   bUsePreviewValueAsDefault=true
   inputName=---------Scale Over Velocity Parameters---------
   inputType=FunctionInput_StaticBool
   sortPriority=10
  }
  n1 = FunctionInput() {bUsePreviewValueAsDefault=true, inputName=Maximum Scale Factor, inputType=FunctionInput_Scalar, sortPriority=15, previewValue=<2 0 0 1>}
  n2 = ParticleSpeed()
  n3 = FunctionInput() {
   bUsePreviewValueAsDefault=true
   description=The particle must go this fast before beginning to warp along the velocity vector.
   inputName=Minimum Speed
   inputType=FunctionInput_Scalar
   sortPriority=12
   previewValue=<10 0 0 1>
  }
  n4 = FunctionInput() {
   bUsePreviewValueAsDefault=true
   description="After reaching the minimum speed, the particle will start to warp until this speed is reached (in addition to the minimum speed)."
   inputName=Max Speed
   inputType=FunctionInput_Scalar
   sortPriority=13
   previewValue=<30 0 0 1>
  }
  n5 = Divide(
   A=Subtract(
    A=FunctionInput(Preview=@n2) {bUsePreviewValueAsDefault=true, inputName=Particle Speed, inputType=FunctionInput_Scalar, sortPriority=11}
    B=@n3
   )
   B=@n4
  )
  n6 = FunctionInput(Preview=LinearInterpolate(
    A=FunctionInput() {bUsePreviewValueAsDefault=true, inputName=Minimum Scale Factor, inputType=FunctionInput_Scalar, sortPriority=14}
    B=@n1
    Alpha=Clamp(None=@n5)
   )) {
   bUsePreviewValueAsDefault=true
   description="Note: By default, an equation using the \"Scale Over Velocity Parameters\" sets this value. Plugging an input into this field will allow one to manually control the scale factor."
   inputName=Scale Along Velocity Vector
   inputType=FunctionInput_Scalar
   sortPriority=9
  }
  n7 = Multiply(A=FunctionInput(Preview=ParticleDirection()) {bUsePreviewValueAsDefault=true, inputName=Particle Direction}, B=@n6)
  n8 = Transform(None=Multiply(
    A=Constant3Vector() {constant=<rgba 1 1 0 0>}
    B=Transform(None=@n7) {transformSourceType=TRANSFORMSOURCE_World, transformType=TRANSFORM_Camera}
   )) {desc=3d velocity vector on x and y, transformSourceType=TRANSFORMSOURCE_Camera}
  n9 = TextureCoordinate()
  n10 = ParticleSize()
  n11 = MaterialFunctionCall(Float2=Multiply(
    A=Subtract(A=@n9) {constB=0.5}
    B=FunctionInput(Preview=@n10) {bUsePreviewValueAsDefault=true, inputName=Particle Size, inputType=FunctionInput_Vector2}
   )) {materialFunction=&BreakOutFloat2Components}
  n12 = Constant3Vector() {constant=<rgba 1 0 0 0>}
  n13 = Constant3Vector() {constant=<rgba 0 -1 0 0>}
  n14 = Add(
   A=Multiply(A=@n11.R, B=Transform(None=@n12) {transformSourceType=TRANSFORMSOURCE_Camera})
   B=Multiply(A=@n11.G, B=Transform(None=@n13) {transformSourceType=TRANSFORMSOURCE_Camera})
  )
  n15 = Add(A=@n14, B=Multiply(A=@n8, B=DotProduct(A=@n14, B=@n8)))
  n16 = FunctionInput(Preview=ParticlePositionWS()) {bUsePreviewValueAsDefault=true, inputName=Particle Position, sortPriority=-1}
  n17 = FunctionInput(Preview=WorldPosition().XYZ) {
   bUsePreviewValueAsDefault=true
   description=World position must be subtracted from the World Position Offset output before being inserted into the World Position Offset pin. This input allows advanced users to override the world position value that's subtracted from the output.
   inputName=World Position to Negate
   sortPriority=8
  }
  FunctionOutput(None=Subtract(A=Add(A=@n15, B=@n16), B=@n17)) {outputName=World Position Offset, sortPriority=3}
  n18 = FunctionInput(Preview=TextureSample() {samplerType=SAMPLERTYPE_Normal, texture=/Engine/Functions/Engine_MaterialFunctions02/ExampleContent/Textures/SphereRenderNormalsMap}.RGB) {
   bUsePreviewValueAsDefault=true
   description=These normals will be transformed from view space to world space. Please ensure that your normal texture's tangent space isn't rotated or skewed.
   inputName=Normals
   sortPriority=4
  }
  n19 = Constant3Vector() {constant=<rgba 1 -1 -1 0>}
  FunctionOutput(None=Transform(None=Multiply(A=@n18, B=@n19)) {transformSourceType=TRANSFORMSOURCE_View}) {
   description="These normals are in world space. Ensure that tangent space normals are unchecked in your material's details panel. "
   outputName=World Space Normal
   sortPriority=2
  }
  FunctionOutput(None=FunctionInput(Preview=TextureSample() {texture=/Engine/Functions/Engine_MaterialFunctions02/ExampleContent/Textures/SphereRenderMask}.R) {bUsePreviewValueAsDefault=true, inputName=Opacity Mask, sortPriority=7}) {outputName=Opacity Mask, sortPriority=1}
 ]
}
