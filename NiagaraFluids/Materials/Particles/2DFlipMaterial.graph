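# Translucent, lit Niagara sprite material for particles in a 2D FLIP fluid simulation from Niagara Fluids.
# Color comes from particle color or a blue fallback parameter, chosen by the UseParticleColor static switch.
# The Sprite_Ellipsoid function's opacity mask, scaled by an Opacity parameter, shapes each sprite.
# A Whitewater parameter multiplies particle speed for a velocity-driven term, and lighting uses the volumetric directional translucency mode.
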Material /NiagaraFluids/Materials/Particles/2DFlipMaterial
graph {
 properties {
  bUsedWithNiagaraSprites=true
  blendMode=BLEND_Translucent
  pixelDepthOffsetMode=PDOM_Legacy
  translucencyLightingMode=TLM_VolumetricDirectional
  usedShadingModels=Default Lit
 }
 nodes [
  Constant() {r=0.2}
  Constant() {r=1}
  Constant()
  Multiply(A=ParticleSpeed(), B=ScalarParameter() {parameterName=Whitewater})
  Multiply(
   A=MaterialFunctionCall!(
    Particle Position=ParticlePositionWS()
    Particle Direction
    Particle Size
    Normals
    Opacity Mask
    World Position to Negate
    Scale Along Velocity Vector
    ---------Scale Over Velocity Parameters---------
    Particle Speed
    Minimum Speed
    Max Speed
    Minimum Scale Factor
    Maximum Scale Factor
   ) {materialFunction=/Engine/Functions/Engine_MaterialFunctions02/WorldPositionOffset/Sprite_Ellipsoid}.Opacity Mask
   B=ScalarParameter() {defaultValue=0.7, parameterName=Opacity}
  )
  StaticSwitchParameter(True=ParticleColor().RGB, False=VectorParameter() {defaultValue=<rgba 0 0.33 1 0>}.RGB) {defaultValue=true, parameterName=UseParticleColor}
 ]
}
