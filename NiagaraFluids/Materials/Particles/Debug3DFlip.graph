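# Masked, Default Lit Niagara sprite material in Niagara Fluids, named for debug display of 3D FLIP particles.
# The engine's Sprite_Ellipsoid function takes each particle's world position and a size from SizeMult times dx, shaping sprites as spheroids.
# Its scale along velocity input is a zero constant, and particle direction is unconnected.
# A blue-default Base Particle Color parameter is multiplied by particle color to tint them.
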Material /NiagaraFluids/Materials/Particles/Debug3DFlip.Debug3DFLIP
graph {
 expressions {
  MaterialExpressionConstant_0 {class=MaterialExpressionConstant}
  MaterialExpressionConvert_1 {
   class=MaterialExpressionConvert
   inputs [
    None {from=MaterialExpressionMultiply_1}
    None {from=MaterialExpressionMultiply_1}
   ]
   properties {
    nodeName=Make Vector2
    convertInputs [
     {
      type=Scalar
      expressionInput {
       expression=&self:MaterialExpressionMultiply_1
       inputName=None
       mask=0
       maskA=0
       maskB=0
       maskG=0
       maskR=0
       outputIndex=0
      }
     }
     {
      type=Scalar
      expressionInput {
       expression=&self:MaterialExpressionMultiply_1
       inputName=None
       mask=0
       maskA=0
       maskB=0
       maskG=0
       maskR=0
       outputIndex=0
      }
     }
    ]
    convertOutputs [{type=Vector2}]
   }
  }
  MaterialExpressionMaterialFunctionCall_0 {
   class=MaterialExpressionMaterialFunctionCall
   inputs [
    Particle Position {from=MaterialExpressionParticlePositionWS_0}
    Particle Direction {from=""}
    Particle Size {from=MaterialExpressionConvert_1}
    Normals {from=""}
    Opacity Mask {from=""}
    World Position to Negate {from=""}
    Scale Along Velocity Vector {from=MaterialExpressionConstant_0}
    ---------Scale Over Velocity Parameters--------- {from=""}
    Particle Speed {from=""}
    Minimum Speed {from=""}
    Max Speed {from=""}
    Minimum Scale Factor {from=""}
    Maximum Scale Factor {from=""}
   ]
   properties {materialFunction=/Engine/Functions/Engine_MaterialFunctions02/WorldPositionOffset/Sprite_Ellipsoid}
  }
  MaterialExpressionMultiply_0 {
   class=MaterialExpressionMultiply
   inputs [
    A {from=MaterialExpressionVectorParameter_0, output=RGB}
    B {from=MaterialExpressionParticleColor_0, output=RGB}
   ]
  }
  MaterialExpressionMultiply_1 {
   class=MaterialExpressionMultiply
   inputs [
    A {from=MaterialExpressionScalarParameter_0}
    B {from=MaterialExpressionScalarParameter_1}
   ]
  }
  MaterialExpressionParticleColor_0 {class=MaterialExpressionParticleColor, inputs [], properties {}}
  MaterialExpressionParticlePositionWS_0 {class=MaterialExpressionParticlePositionWS, inputs [], properties {}}
  MaterialExpressionScalarParameter_0 {class=MaterialExpressionScalarParameter, properties {parameterName=SizeMult}}
  MaterialExpressionScalarParameter_1 {class=MaterialExpressionScalarParameter, properties {parameterName=dx}}
  MaterialExpressionVectorParameter_0 {class=MaterialExpressionVectorParameter, properties {parameterName=Base Particle Color, defaultValue=<rgba 0 0.1715 0.9271 0>}}
 }
 properties {bUsedWithNiagaraSprites=true, blendMode=BLEND_Masked, pixelDepthOffsetMode=PDOM_Legacy, usedShadingModels=Default Lit}
}
