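# Translucent, two-sided lit material for mist particles on Niagara sprites and ribbons in NiagaraFluids.
# A soft radial sphere mask is multiplied by a fade, a small scale factor and tiling noise to form a faint, broken-up puff.
# The Use Particle Color switch takes fade and tint from particle color, or from a relative-lifetime falloff and a pale blue parameter.
# A tiling detail normal shares the scaled noise UVs.
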Material /NiagaraFluids/Materials/Particles/MistMaterial
graph {
 properties {
  bEnableResponsiveAA=true
  bTangentSpaceNormal=false
  bUsedWithNiagaraRibbons=true
  bUsedWithNiagaraSprites=true
  bUsedWithParticleSprites=true
  blendMode=BLEND_Translucent
  pixelDepthOffsetMode=PDOM_Legacy
  translucencyLightingMode=TLM_SurfacePerPixelLighting
  twoSided=true
  usedShadingModels=Default Lit
 }
 nodes [
  n1 = Multiply(A=TextureCoordinate(), B=ScalarParameter() {defaultValue=0.25})
  n2 = TextureSample(UVs=@n1) {texture=/Engine/MapTemplates/TilingNoise05}
  Multiply(B=@n2.RGB)
  n3 = ParticleColor()
  n4 = ParticleRelativeTime()
  n5 = StaticSwitchParameter(True=@n3.A, False=Power(Base=OneMinus(None=@n4), Exp=ScalarParameter() {defaultValue=1, parameterName=Opacity Falloff})) {parameterName=Use Particle Color}
  n6 = SphereMask(
   A=TextureCoordinate()
   B=Constant() {r=0.5}
   Radius=ScalarParameter() {defaultValue=0.5, parameterName=Radius}
   Hardness=ScalarParameter() {defaultValue=0.5, parameterName=Hardness}
  )
  Multiply(A=Multiply(A=ScalarParameter() {defaultValue=0.0025, parameterName=Param_1}, B=Multiply(A=@n5, B=@n6)) {constA=0.05}, B=@n2.R)
  StaticSwitchParameter(True=@n3.RGB, False=VectorParameter() {parameterName=Color, defaultValue=<rgba 0.6334 0.8192 0.9427 1>}.RGB) {parameterName=Use Particle Color}
  TextureSample(UVs=@n1) {samplerType=SAMPLERTYPE_Normal, texture=/Engine/Engine_MI_Shaders/T_Base_Tile_DetailNormal}
  VectorParameter() {parameterName=MistColor, defaultValue=<rgba 0.6334 0.8192 0.9427 1>}
 ]
}
