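# Translucent, two-sided lit particle material for Niagara Fluids whitewater, usable on sprites and ribbons.
# A static switch takes color and fade from either particle color or a color parameter with a lifetime falloff.
# Opacity multiplies that fade by a soft radial mask, noise texture and multiplier.
# Dynamic parameter 1 offsets the scaled UVs shared by noise and a detail normal sample, with per-particle random added for the noise.
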Material /NiagaraFluids/Materials/Particles/WhitewaterMaterial
graph {
 properties {
  bEnableResponsiveAA=true
  bTangentSpaceNormal=false
  bUsedWithNiagaraRibbons=true
  bUsedWithNiagaraSprites=true
  bUsedWithParticleSprites=true
  blendMode=BLEND_Translucent
  pixelDepthOffsetMode=PDOM_Legacy
  translucencyLightingMode=TLM_SurfacePerPixelLighting
  twoSided=true
  usedShadingModels=Default Lit
 }
 nodes [
  n1 = ParticleColor()
  n2 = ParticleRelativeTime()
  n3 = StaticSwitchParameter(True=@n1.A, False=Power(Base=OneMinus(None=@n2), Exp=ScalarParameter() {defaultValue=1, parameterName=Opacity Falloff})) {parameterName=Use Particle Color}
  n4 = SphereMask(
   A=TextureCoordinate()
   B=Constant() {r=0.5}
   Radius=ScalarParameter() {defaultValue=0.5, parameterName=Radius}
   Hardness=ScalarParameter() {defaultValue=0.5, parameterName=Hardness}
  )
  n5 = DynamicParameter()
  n6 = Multiply(
   A=Add(
    A=TextureCoordinate()
    B=MaterialFunctionCall(X=@n5.Param1, Y=@n5.Param1) {materialFunction=&MakeFloat2}.Result
   )
   B=ScalarParameter() {defaultValue=0.5}
  )
  n7 = Add(A=@n6, B=ParticleRandom())
  Multiply(
   A=Multiply(A=Multiply(A=@n3, B=@n4), B=TextureSample(UVs=@n7) {texture=/Engine/MaterialTemplates/Textures/T_Noise01}.RGB)
   B=ScalarParameter() {defaultValue=5, parameterName=OpacityMult}
  )
  ScalarParameter() {defaultValue=1000, parameterName=Param_1}
  StaticSwitchParameter(True=@n1.RGB, False=VectorParameter() {parameterName=Color, defaultValue=<rgba 0.6334 0.8192 0.9427 1>}.RGB) {parameterName=Use Particle Color}
  TextureSample(UVs=@n6) {samplerType=SAMPLERTYPE_Normal, texture=/Engine/Engine_MI_Shaders/T_Base_Tile_DetailNormal}
 ]
}
